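; Storm Aegis boon gameplay ability that grants shield to the player when they take storm damage.
; On activation it prepares a boon effect spec and listens for damage-received events on the avatar pawn.
; Events whose instigator tags include the storm damage tag, while shield is below max, feed their magnitude into the SetByCaller shield value and apply the effect to the owner.
; Ending the ability ends the async listener.
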(event Ability|EventActivateAbility
  (|CallParentFunction)
  (Variables|Default|SetPlayerPawn (CallFunction|GetAvatarasFortPlayerPawn))
  (Variables|Default|SetHealingSpec (Ability|MakeOutgoingGameplayEffectSpec "/CherrySmokeBoonsGameplay/Gameplay/StormAegisBoon/GE_PA_StormAegisBoon.GE_PA_StormAegisBoon_C"))
  (bind (_asynctaskproxy _payload) (Ability|Async|WaitGameplayEventToActor (Variables|Default|GetPlayerPawn) "(TagName=\"Event.Damage.Received\")")
    (:then
      (Variables|Default|SetAsyncActionA _asynctaskproxy))
    (:EventReceived
      (if (GameplayTags|HasTag (.InstigatorTags (Utilities|Struct|BreakGameplayEventData _payload)) (Variables|Default|GetStormDamageTag))
        (if (CallFunction|CheckShieldValue)
          (bind _returnvalue (Ability|GameplayEffect|AssignTagSetbyCallerMagnitude (Variables|Default|GetHealingSpec) "(TagName=\"SetByCaller.Shield\")" (.EventMagnitude (Utilities|Struct|BreakGameplayEventData _payload))))
          (Ability|ApplyGameplayEffectSpecToOwner _returnvalue))))))

(event Ability|EventOnEndAbility
  (params (bWasCancelled :type Boolean))
  (|CallParentFunction bWasCancelled)
  (Ability|Async|EndAction (Variables|Default|GetAsyncActionA)))

(fn CheckShieldValue
  (return (< (FortPawn|GetShield (Ability|GetActivatingPawn)) (FortPawn|GetMaxShield (Ability|GetActivatingPawn)))))
